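# Conversation database for the Season 20 Week 9 Quest 3 archive server, presented as a terminal-style browser of character CSV files.
# Chained Speech nodes page through six entries via "Next Entry", wrapping with ReturnToStart, while "Examine Entry" shows a fake `cat` readout and "Close Entry" goes Back.
# The Imagined and Order files return "No such file" with an error cue.
# Examining the_order.csv also fires the quest prop success side effect and has no outputs.
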ConversationDatabase /BattlePassS20/Conversations/Conversation_S20_W9_Q3_ArchiveServer
entry_points [{tag=Conversation.Entry.Season20.Narrative.ArchiveServer, destinations [#1]}]
nodes [
 ConversationEntryPointNode_7 {
  id=#1
  class=ConversationEntryPointNode
  outputs [#2]
  properties {entryTag {tagName=Conversation.Entry.Season20.Narrative.ArchiveServer}}
 }
 Speech_C_12 {
  id=#2
  class=Speech_C
  outputs [#3, #4]
  properties {generalConfig {defaultMessage=the_foundation.csv}}
 }
 Speech_C_5 {
  id=#4
  class=Speech_C
  outputs [#5, #6]
  properties {softTaskIcon=/Game/UI/Foundation/Textures/Icons/Replay/T-Icon-Replay-Forward-L, generalConfig {defaultMessage=the_origin.csv}}
  sub_nodes [ConversationChoiceNode_6 {class=ConversationChoiceNode, properties {defaultChoiceDisplayText=Next Entry}}]
 }
 Speech_C_6 {
  id=#6
  class=Speech_C
  outputs [#7, #8]
  properties {softTaskIcon=&T-Icon-Replay-Forward-L, generalConfig {defaultMessage=the_paradigm.csv}}
  sub_nodes [ConversationChoiceNode_2 {class=ConversationChoiceNode, properties {defaultChoiceDisplayText=Next Entry}}]
 }
 Speech_C_7 {
  id=#8
  class=Speech_C
  outputs [#9, #10]
  properties {softTaskIcon=&T-Icon-Replay-Forward-L, generalConfig {defaultMessage=peely.csv}}
  sub_nodes [ConversationChoiceNode_26 {class=ConversationChoiceNode, properties {defaultChoiceDisplayText=Next Entry}}]
 }
 Speech_C_15 {
  id=#10
  class=Speech_C
  outputs [#11, #12]
  properties {softTaskIcon=&T-Icon-Replay-Forward-L, generalConfig {defaultMessage=the_imagined.csv}}
  sub_nodes [ConversationChoiceNode_30 {class=ConversationChoiceNode, properties {defaultChoiceDisplayText=Next Entry}}]
 }
 Speech_C_17 {
  id=#12
  class=Speech_C
  outputs [#13, #14]
  properties {softTaskIcon=&T-Icon-Replay-Forward-L, generalConfig {defaultMessage=the_order.csv}}
  sub_nodes [ConversationChoiceNode_8 {class=ConversationChoiceNode, properties {defaultChoiceDisplayText=Next Entry}}]
 }
 ReturnToStart_C_1 {
  id=#14
  class=ReturnToStart_C
  properties {softTaskIcon=&T-Icon-Replay-Forward-L}
  sub_nodes [ConversationChoiceNode_9 {class=ConversationChoiceNode, properties {defaultChoiceDisplayText=Next Entry}}]
 }
 Speech_C_19 {
  id=#13
  class=Speech_C
  outputs []
  properties {
   choiceSound=/Game/Athena/Sounds/NPCs/Conversation/Conversation_Error_Cue
   hoverChoiceSound=/Game/Athena/Sounds/NPCs/Conversation/Conversation_Hover_Cue
   softTaskIcon=/Game/UI/Foundation/Textures/Icons/Replay/T-Icon-Replay-Plus-L
   generalConfig {defaultMessage="$ cat the_order.csv\r\nNo such file or directory."}
  }
  sub_nodes [
   ConvoSideEffect_QuestPropActor_ConvoSuccess_C_1 {class=ConvoSideEffect_QuestPropActor_ConvoSuccess_C}
   ConversationChoiceNode_5 {class=ConversationChoiceNode, properties {defaultChoiceDisplayText=Examine Entry}}
  ]
 }
 Speech_C_16 {
  id=#11
  class=Speech_C
  outputs [#15]
  properties {
   choiceSound=&Conversation_Error_Cue
   hoverChoiceSound=&Conversation_Hover_Cue
   softTaskIcon=&T-Icon-Replay-Plus-L
   generalConfig {defaultMessage="$ cat the_imagined.csv\r\nNo such file or directory."}
  }
  sub_nodes [ConversationChoiceNode_32 {class=ConversationChoiceNode, properties {defaultChoiceDisplayText=Examine Entry}}]
 }
 Back_C_11 {
  id=#15
  class=Back_C
  properties {softTaskIcon=/Game/UI/Foundation/Textures/Icons/Replay/T-Icon-Replay-Minus-L}
  sub_nodes [ConversationChoiceNode_31 {class=ConversationChoiceNode, properties {defaultChoiceDisplayText=Close Entry}}]
 }
 Speech_C_14 {
  id=#9
  class=Speech_C
  outputs [#16]
  properties {
   choiceSound=/Game/Athena/Sounds/NPCs/Conversation/Conversation_Confirm_Cue
   hoverChoiceSound=&Conversation_Hover_Cue
   softTaskIcon=&T-Icon-Replay-Plus-L
   generalConfig {defaultMessage="$ cat peely.csv\r\nDriving Infractions:    1,274\r\nFines Owed:     503245 Bars"}
  }
  sub_nodes [ConversationChoiceNode_28 {class=ConversationChoiceNode, properties {defaultChoiceDisplayText=Examine Entry}}]
 }
 Back_C_9 {
  id=#16
  class=Back_C
  properties {softTaskIcon=&T-Icon-Replay-Minus-L}
  sub_nodes [ConversationChoiceNode_27 {class=ConversationChoiceNode, properties {defaultChoiceDisplayText=Close Entry}}]
 }
 Speech_C_10 {
  id=#7
  class=Speech_C
  outputs [#17]
  properties {
   choiceSound=&Conversation_Confirm_Cue
   hoverChoiceSound=&Conversation_Hover_Cue
   softTaskIcon=&T-Icon-Replay-Plus-L
   generalConfig {defaultMessage="$ cat the_paradigm.csv\r\nStatus:    Unknown\r\nAdvisement:     Capture or Eliminate"}
  }
  sub_nodes [ConversationChoiceNode_24 {class=ConversationChoiceNode, properties {defaultChoiceDisplayText=Examine Entry}}]
 }
 Back_C_4 {
  id=#17
  class=Back_C
  properties {softTaskIcon=&T-Icon-Replay-Minus-L}
  sub_nodes [ConversationChoiceNode_23 {class=ConversationChoiceNode, properties {defaultChoiceDisplayText=Close Entry}}]
 }
 Speech_C_13 {
  id=#5
  class=Speech_C
  outputs [#18]
  properties {
   choiceSound=&Conversation_Confirm_Cue
   hoverChoiceSound=&Conversation_Hover_Cue
   softTaskIcon=&T-Icon-Replay-Plus-L
   generalConfig {defaultMessage="$ cat the_origin.csv\r\nStatus:    Field Active\r\nAdvisement:     Eliminate on sight"}
  }
  sub_nodes [ConversationChoiceNode_20 {class=ConversationChoiceNode, properties {defaultChoiceDisplayText=Examine Entry}}]
 }
 Back_C_6 {
  id=#18
  class=Back_C
  properties {softTaskIcon=&T-Icon-Replay-Minus-L}
  sub_nodes [ConversationChoiceNode_0 {class=ConversationChoiceNode, properties {defaultChoiceDisplayText=Close Entry}}]
 }
 Speech_C_4 {
  id=#3
  class=Speech_C
  outputs [#19]
  properties {
   choiceSound=&Conversation_Confirm_Cue
   hoverChoiceSound=&Conversation_Hover_Cue
   softTaskIcon=&T-Icon-Replay-Plus-L
   generalConfig {defaultMessage="$ cat the_foundation.csv\r\nStatus:    In Reserve (Sanctuary)\r\nAdvisement:     Do not approach"}
  }
  sub_nodes [ConversationChoiceNode_13 {class=ConversationChoiceNode, properties {defaultChoiceDisplayText=Examine Entry}}]
 }
 Back_C_2 {
  id=#19
  class=Back_C
  properties {softTaskIcon=&T-Icon-Replay-Minus-L}
  sub_nodes [ConversationChoiceNode_12 {class=ConversationChoiceNode, properties {defaultChoiceDisplayText=Close Entry}}]
 }
]
